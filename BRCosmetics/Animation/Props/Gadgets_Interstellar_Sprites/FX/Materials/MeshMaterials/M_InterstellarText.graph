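# M_InterstellarText is a masked, default-lit material for Niagara mesh particles, filed with the Interstellar Sprites gadget prop effects, apparently for text meshes.
# Its graph holds a single Particle Color node, with no connection to a material output shown.
# It has no textures, parameters or other logic.
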Material /BRCosmetics/Animation/Props/Gadgets_Interstellar_Sprites/FX/Materials/MeshMaterials/M_InterstellarText
graph {
 properties {bUsedWithNiagaraMeshParticles=true, blendMode=BLEND_Masked, pixelDepthOffsetMode=PDOM_Legacy, usedShadingModels=Default Lit}
 nodes [ParticleColor()]
}
