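# Masked, lit Niagara particle material for the Interstellar Sprites gadget cosmetic, showing frames from a SubUV sprite sheet.
# A dynamic parameter drives frame, grid size and palette selection, with a debug frame offset.
# Green-channel thresholds form wing and face masks for compositing blue/purple/red palettes of wing, face and eye colors.
# Static switches choose particle color, palette or raw texture color; texture alpha is multiplied by particle alpha.
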Material /BRCosmetics/Animation/Props/Gadgets_Interstellar_Sprites/FX/Materials/MeshMaterials/M_Interstellar_SubUV_Mesh
graph {
 properties {
  bEnableResponsiveAA=true
  bOutputTranslucentVelocity=true
  bUsedWithNiagaraMeshParticles=true
  bUsedWithNiagaraSprites=true
  blendMode=BLEND_Masked
  opacityMaskClipValue=0.01
  pixelDepthOffsetMode=PDOM_Legacy
  usedShadingModels=Default Lit
 }
 nodes [
  n1 = DynamicParameter() {parameterIndex=1, paramNames [Variant_SubUVIndex, ParticleUV_x, ParticleUV_y, Variant_NumSprites]}
  Multiply(A=OneMinus(None=Saturate(None=@n1.Variant_SubUVIndex)), B=VertexColor() {desc=blue for sides or red for front}.B)
  NamedRerouteDeclaration(None=MaterialFunctionCall(
    A=VectorParameter() {parameterName=WingColor_Blue, defaultValue=<rgba 0.0194 0.0666 0.3372 1>}.RGB
    B=VectorParameter() {parameterName=WingColor_Purple, defaultValue=<rgba 0.0194 0.0666 0.3372 1>}.RGB
    C=VectorParameter() {parameterName=WingColor_Red, defaultValue=<rgba 0.0194 0.117 0.4179 1>}.RGB
    Alpha=DynamicParameter() {defaultValue=<rgba 0 1 1 4>, paramNames [SubUVIndex, PaletteSelection, ParticleUV_y, NumSprites]}.PaletteSelection
   ) {materialFunction=&Lerp_3Color}.Result) {name=Palette_Wing, nodeColor=<rgba 0.1953 1 0 1>}
  NamedRerouteDeclaration(None=MaterialFunctionCall(
    A=VectorParameter() {parameterName=FaceColor_Blue, defaultValue=<rgba 0.0194 0.117 0.4179 1>}.RGB
    B=VectorParameter() {parameterName=FaceColor_Purple, defaultValue=<rgba 0.4988 0.117 0.626 1>}.RGB
    C=VectorParameter() {parameterName=FaceColor_Red, defaultValue=<rgba 0.7991 0.0176 0.0091 1>}.RGB
    Alpha=DynamicParameter() {defaultValue=<rgba 0 1 1 4>, paramNames [SubUVIndex, PaletteSelection, ParticleUV_y, NumSprites]}.PaletteSelection
   ) {materialFunction=&Lerp_3Color}.Result) {name=Palette_Face, nodeColor=<rgba 0 1 0.4844 1>}
  n2 = DynamicParameter() {defaultValue=<rgba 0 1 1 4>, paramNames [SubUVIndex, PaletteSelection, ParticleUV_y, NumSprites]}
  n3 = MaterialFunctionCall!(
   Texture=TextureObjectParameter() {parameterName=TextureObject Parm, texture=/BRCosmetics/Animation/Props/Gadgets_Interstellar_Sprites/FX/Textures/T_InterstellarSheet}
   UVs=TextureCoordinate()
   SubImages=AppendVector(A=@n2.NumSprites, B=@n2.NumSprites)
   Frame=Add(A=ScalarParameter() {parameterName=DebugSubUVOffset}, B=@n2.SubUVIndex)
  ) {materialFunction=/Engine/Functions/Engine_MaterialFunctions03/Texturing/SubUV_Function_MipDerivative}
  n4 = ComponentMask(None=@n3.RGB) {g=true}
  n5 = Constant() {r=1}
  NamedRerouteDeclaration(None=If(A=@n4, B=Constant() {r=0.1}, "A > B"=Constant(), "A == B"=@n5, "A < B"=@n5)) {name=Wings, nodeColor=<rgba 0.4766 1 0 1>}
  NamedRerouteDeclaration(None=MaterialFunctionCall(
    A=VectorParameter() {parameterName=EyeColor_Blue, defaultValue=<rgba 1 0 0 1>}.RGB
    B=VectorParameter() {parameterName=EyeColor_Purple, defaultValue=<rgba 1 0 0 1>}.RGB
    C=VectorParameter() {parameterName=EyeColor_Red, defaultValue=<rgba 0.8714 0.8308 0.0595 1>}.RGB
    Alpha=DynamicParameter() {defaultValue=<rgba 0 1 1 4>, paramNames [SubUVIndex, PaletteSelection, ParticleUV_y, NumSprites]}.PaletteSelection
   ) {materialFunction=&Lerp_3Color}.Result) {name=Palette_Eye, nodeColor=<rgba 0.2422 1 0 1>}
  n6 = Constant() {r=1}
  NamedRerouteDeclaration(None=If(A=@n4, B=Constant() {r=0.6}, "A > B"=@n6, "A == B"=@n6, "A < B"=Constant())) {name=Face, nodeColor=<rgba 1 0 0.5391 1>}
  n7 = Multiply(A=NamedRerouteUsage(), B=NamedRerouteUsage())
  n8 = Multiply(A=NamedRerouteUsage(), B=NamedRerouteUsage())
  n9 = Multiply(A=OneMinus(None=NamedRerouteUsage()), B=OneMinus(None=NamedRerouteUsage()))
  n10 = NamedRerouteUsage()
  NamedRerouteDeclaration(None=Add(A=Add(A=@n7, B=@n8), B=Multiply(A=@n9, B=@n10))) {name=Palettized, nodeColor=<rgba 0 0.3672 1 1>}
  n11 = ParticleColor()
  Reroute(None=Multiply(A=@n3.Alpha, B=@n11.A))
  StaticSwitchParameter(True=@n11.RGB, False=StaticSwitchParameter(True=NamedRerouteUsage(), False=@n3.RGB) {parameterName=UsePaletteColors}) {parameterName=Use Particle Color}
 ]
}
